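# Translucent unlit beam-trail material for the grappling hook rope, colored by particle color through an emissive adjustment function.
# Two copies of a packed edge texture pan along U with opposite UV offsets, then are sharpened by power and floor into hard strands.
# Dynamic parameters drive how the V offset varies along the beam.
# A sine mask tapers the strands at both ends, a "Straight?" lerp selects them over a ramp texture, and particle alpha scales the result.
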Material /FNGameplayCues/Athena/Items/Consumables/Grapple/M_GrapplingHook_RopeBeam
graph {
 properties {
  bEnableResponsiveAA=true
  bUsedWithBeamTrails=true
  blendMode=BLEND_Translucent
  pixelDepthOffsetMode=PDOM_Legacy
  shadingModel=MSM_Unlit
  usedShadingModels=Unlit
 }
 nodes [
  n1 = ParticleColor()
  MaterialFunctionCall(Emissive=@n1.RGB) {materialFunction=&MF_FX_EmissiveAdjustments}
  MaterialFunctionCall(Speed=ScalarParameter() {defaultValue=1, parameterName=Pulse Speed}) {materialFunction=&DebugTimeSine}
  n2 = TextureSample() {texture=/Game/Global/Textures/Default/Masks_and_Gradients/T_Water_Top_Ramp}
  n3 = TextureCoordinate()
  n4 = ScalarParameter() {parameterName=V_Offset}
  n5 = ScalarParameter() {defaultValue=0.025, parameterName=V_Offset_1}
  n6 = DynamicParameter() {defaultValue=<rgba 301 1 1 1>, paramNames [Utile, VOffsetControl, Param3, Param4]}
  n7 = ComponentMask(None=TextureCoordinate()) {r=true}
  n8 = DynamicParameter() {defaultValue=<rgba 301 1 1 1>, paramNames [Utile, VOffsetControl, Param3, Param4]}
  n9 = AppendVector(
   A=ScalarParameter() {defaultValue=0.0024, parameterName=U_Offset}
   B=LinearInterpolate(
    A=LinearInterpolate(A=@n4, B=@n5, Alpha=@n6.VOffsetControl)
    B=LinearInterpolate(A=@n5, B=@n4, Alpha=@n6.VOffsetControl)
    Alpha=Multiply(A=@n7, B=@n8.Utile)
   )
  )
  n10 = Panner(Coordinate=Add(A=@n3, B=@n9)) {speedX=-1}
  n11 = ScalarParameter() {defaultValue=3, parameterName=Contrast}
  n12 = Power(Base=Multiply(A=TextureSample(UVs=@n10) {texture=/Game/Effects/Fort_Effects/Textures/Taker/T_PortalEdge_02_Packed}.R) {constB=100}, Exp=@n11)
  n13 = Panner(Coordinate=Add(A=@n3, B=Multiply(A=@n9) {constB=-1})) {speedX=-1}
  n14 = Power(Base=Multiply(A=TextureSample(UVs=@n13) {texture=&T_PortalEdge_02_Packed}.R) {constB=100}, Exp=@n11)
  n15 = Saturate(None=Add(A=Floor(None=@n12), B=Floor(None=@n14)))
  n16 = TextureCoordinate()
  n17 = Power(Base=Sine(None=ComponentMask(None=@n16) {r=true}) {period=2}, Exp=ScalarParameter() {defaultValue=5})
  Multiply(
   A=@n1.A
   B=LinearInterpolate(A=Power(Base=@n2.R), B=Multiply(A=@n15, B=@n17), Alpha=ScalarParameter() {defaultValue=1, parameterName=Straight?})
  )
  ScalarParameter() {defaultValue=301.1938, parameterName=UTile}
 ]
}
